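# One-sided masked master material for static-lit and skeletal meshes, converting UE3-style Phong diffuse, specular and power inputs to PBR.
# Two diffuse, specular and normal sets blend by a texture or world-vector normal falloff, optionally modulated.
# Static switches add bump offset, detail normals, variation, fake reflections, Fresnel rim and emissive.
# Color parameters are mostly squared, and UE3 transmission branches remain as noted leftovers.
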Material /Engine/Engine_MI_Shaders/M_ES_Phong_Masked_Master_1sided
graph {
 properties {
  bUsedWithMeshDeformer=true
  bUsedWithSkeletalMesh=true
  bUsedWithStaticLighting=true
  blendMode=BLEND_Masked
  pixelDepthOffsetMode=PDOM_Legacy
  usedShadingModels=Default Lit
 }
 nodes [
  n1 = StaticSwitchParameter(
   True=AppendVector(A=ScalarParameter() {defaultValue=1, parameterName=Tex_Scale_01_X}, B=ScalarParameter() {defaultValue=1, parameterName=Tex_Scale_01_Y})
   False=Constant() {r=1}
  ) {parameterName=NonUniformScale}
  n2 = TextureCoordinate() {coordinateIndex=1}
  n3 = TextureSampleParameter2D(UVs=Multiply(A=Multiply(A=@n1, B=@n2), B=ScalarParameter() {defaultValue=1, parameterName=BlendModulationScale})) {desc=ModulationTex, parameterName=BlendModulationTexture, texture=&/T_Base_Tile_Specular}
  n4 = StaticComponentMaskParameter(None=AppendVector(A=@n3.RGB, B=@n3.RGB)) {defaultR=true, parameterName=Blend_Modulation_Channel}
  n5 = ScalarParameter() {defaultValue=-0.5, parameterName=Modulation_Offset}
  n6 = Multiply(
   A=Power(Base=Add(A=@n4, B=@n5), Exp=ScalarParameter() {defaultValue=1, parameterName=Modulation_Power})
   B=ScalarParameter() {defaultValue=0.5, parameterName=Modulation_Multiply}
  )
  n7 = TextureCoordinate() {coordinateIndex=1}
  n8 = TextureCoordinate()
  n9 = TextureSampleParameter2D() {parameterName=BumpOffsetTexture, texture=&/T_Reflection}
  n10 = StaticComponentMaskParameter(None=AppendVector(A=@n9.RGB, B=@n9.RGB)) {defaultA=true, parameterName=BumpOffsetChannel}
  n11 = ScalarParameter() {defaultValue=1, parameterName=BumpOffsetHeight}
  n12 = StaticSwitchParameter(True=BumpOffset(Coordinate=@n8, Height=Multiply(A=@n10, B=@n11)), False=@n8) {parameterName=UseBumpoffset}
  n13 = Multiply(
   A=Multiply(A=@n1, B=StaticSwitchParameter(True=@n7, False=@n12) {parameterName=Tex_01_UV2})
   B=ScalarParameter() {defaultValue=1, parameterName=Tex_Scale_01}
  )
  n14 = TextureSampleParameter2D(UVs=StaticSwitchParameter(True=@n13, False=TextureCoordinate()) {defaultValue=true, parameterName=ScaleNormal_01}) {parameterName=Normal_01, samplerType=SAMPLERTYPE_Normal, texture=&/T_Normal}
  n15 = Multiply(
   A=Multiply(A=@n1, B=StaticSwitchParameter(True=@n7, False=@n12) {parameterName=Tex_02_UV2})
   B=ScalarParameter() {defaultValue=1, parameterName=Tex_Scale_02}
  )
  n16 = TextureSampleParameter2D(UVs=StaticSwitchParameter(True=@n15, False=TextureCoordinate()) {defaultValue=true, parameterName=ScaleNormal_02}) {parameterName=Normal_02, samplerType=SAMPLERTYPE_Normal, texture=&T_Normal}
  n17 = StaticSwitchParameter(True=Add(A=@n14.RGB, B=@n16.RGB), False=StaticSwitchParameter(True=@n16.RGB, False=@n14.RGB) {parameterName=BlendUsingNormal2}) {parameterName=BlendUsingCombinedNormal}
  n18 = Constant3Vector() {constant=<rgba 0 0 1 0>}
  n19 = Normalize(VectorInput=Transform(None=StaticSwitchParameter(True=@n17, False=@n18) {parameterName=BlendUsingNormal}))
  n20 = VectorParameter() {parameterName=WorldVector, defaultValue=<rgba 0 0 1 1>}
  n21 = Constant() {r=0.5}
  n22 = Add(A=Multiply(A=DotProduct(A=@n19, B=@n20.RGB), B=@n21), B=@n21)
  n23 = ScalarParameter() {defaultValue=2, parameterName=FalloffExponent}
  n24 = ScalarParameter() {defaultValue=4, parameterName=FalloffMultiply}
  n25 = TextureCoordinate() {coordinateIndex=1}
  n26 = TextureCoordinate()
  n27 = TextureSampleParameter2D(UVs=Multiply(
    A=ScalarParameter() {defaultValue=1, parameterName=BlendAlphaScale}
    B=StaticSwitchParameter(True=@n25, False=@n26) {parameterName=BlendAlphaUV2}
   )) {parameterName=BlendAlphaTex, texture=&/T_Base_Tile_Emmisive}
  n28 = StaticComponentMaskParameter(None=AppendVector(A=@n27.RGB, B=@n27.RGB)) {defaultR=true, parameterName=BlendTexChannel}
  n29 = StaticSwitchParameter(True=Multiply(A=Power(Base=@n28, Exp=@n23), B=@n24), False=@n28) {parameterName=ModifyBlendAlpha}
  n30 = StaticSwitchParameter(True=Multiply(A=Power(Base=@n22, Exp=@n23), B=@n24), False=@n29) {parameterName=EnableWorldVectorBlend}
  n31 = Clamp(None=LinearInterpolate(A=Multiply(A=@n6, B=@n30), B=Add(A=@n6, B=@n30), Alpha=@n30))
  n32 = StaticSwitchParameter(
   True=Clamp(None=StaticSwitchParameter(True=@n31, False=@n30) {parameterName=EnableBlendModulation}) {maxDefault=0.98, minDefault=0.02}
   False=Constant()
  ) {parameterName=BlendSecondTextureSet}
  MaterialFunctionCall!(Power=LinearInterpolate(A=ScalarParameter() {defaultValue=5, parameterName=SpecPower1}, B=ScalarParameter() {defaultValue=25, parameterName=SpecPower2}, Alpha=@n32)) {materialFunction=/Engine/Functions/Engine_MaterialFunctions01/Shading/PowerToRoughness}
  n33 = VectorParameter() {desc=Color, parameterName=FresnelColor, defaultValue=<rgba 1 1 1 5>}
  n34 = Multiply(A=Power(Base=Fresnel(), Exp=ScalarParameter() {defaultValue=1, parameterName=FresnelFalloffExponent}), B=Multiply(A=@n33.RGB, B=@n33.RGB))
  n35 = Add(A=@n34, B=Constant() {r=1})
  n36 = VectorParameter() {parameterName=Diffuse_01_Color, defaultValue=<rgba 1 1 1 1>}
  n37 = TextureSampleParameter2D(UVs=@n13) {parameterName=Diffuse_01, texture=&T_Reflection}
  n38 = VectorParameter() {parameterName=Diffuse_02_Color, defaultValue=<rgba 1 1 1 1>}
  n39 = Multiply(A=@n38.RGB, B=@n38.RGB)
  n40 = Constant() {r=1}
  n41 = Multiply(
   A=Multiply(A=Multiply(A=@n36.RGB, B=@n36.RGB), B=@n37.RGB) {desc=Diffuse 1}
   B=LinearInterpolate(A=@n39, B=Constant() {r=1}, Alpha=StaticSwitchParameter(True=@n29, False=@n40) {parameterName=Colorize Diffuse with Alpha})
  )
  n42 = VectorParameter() {desc=Color, parameterName=VariationColor, defaultValue=<rgba 1 1 1 1>}
  n43 = Multiply(A=@n42.RGB, B=@n42.RGB)
  n44 = StaticSwitchParameter(True=TextureCoordinate() {coordinateIndex=1}, False=TextureCoordinate()) {parameterName=VariationUV2}
  n45 = TextureSampleParameter2D(UVs=Multiply(A=Multiply(A=@n44, B=@n1), B=ScalarParameter() {defaultValue=0.3, desc=VariationScale, parameterName=VariationScale})) {desc=VariationTex, parameterName=VariationTexture, texture=&/T_Base_Tile_Variation}
  n46 = StaticSwitchParameter(
   True=StaticComponentMaskParameter(None=AppendVector(A=@n45.RGB, B=@n45.RGB)) {defaultR=true, parameterName=Variation_Channel}
   False=@n45.RGB
  ) {parameterName=VariationSingleChannel}
  n47 = Power(Base=@n46, Exp=ScalarParameter() {defaultValue=1, desc=VariationPower, parameterName=VariationPower})
  n48 = StaticSwitchParameter(True=Clamp(None=Multiply(A=@n43, B=@n47), Min=Constant(), Max=ScalarParameter() {defaultValue=1, parameterName=VariationMax}), False=@n46) {parameterName=ExposeVariationControls}
  n49 = Multiply(A=@n39, B=TextureSampleParameter2D(UVs=@n15) {parameterName=Diffuse_02, texture=&T_Reflection}.RGB) {desc=Diffuse 2}
  n50 = LinearInterpolate(A=@n41, B=Multiply(A=@n48, B=@n49), Alpha=@n32)
  n51 = LinearInterpolate(A=Multiply(A=@n48, B=@n41), B=@n49, Alpha=@n32)
  n52 = TextureSampleParameter2D() {desc=ReflectionMask, parameterName=ReflectionMask, texture=&/T_Emissive}
  n53 = AppendVector(A=@n52.RGB, B=@n52.RGB)
  n54 = ReflectionVectorWS(CustomWorldNormal=WorldPosition().XYZ) {desc="UE3 used ReflectionVector expression, which did not have an input. "}
  n55 = TextureSampleParameterCube!(UVs=@n54, Apply View MipBias) {parameterName=CubeMap, texture=&/T_Cubemap_01}
  n56 = Multiply(A=@n54, B=ScalarParameter() {defaultValue=0.5, parameterName=Reflection Scale})
  n57 = ScalarParameter() {defaultValue=0.5, parameterName=Reflection Offset}
  n58 = TextureSampleParameter2D(UVs=ComponentMask(None=Add(A=@n56, B=@n57)) {g=true, r=true}) {desc=Reflection, parameterName=reflection, texture=&T_Emissive}
  n59 = VectorParameter() {parameterName=ReflectionColor, defaultValue=<rgba 1 1 1 1>}
  n60 = Multiply(
   A=StaticSwitchParameter(True=StaticComponentMaskParameter(None=@n53) {defaultR=true, parameterName=Reflection_Mask_Channel}, False=@n52.RGB) {parameterName=ReflectionMaskSingleChannel}
   B=Multiply(A=StaticSwitchParameter(True=@n55.RGB, False=@n58.RGB) {parameterName=Reflection Use Cubemap}, B=Multiply(A=@n59.RGB, B=@n59.RGB))
  )
  n61 = LinearInterpolate(A=@n41, B=@n49, Alpha=@n32)
  n62 = StaticSwitchParameter(True=Add(A=@n60, B=@n61), False=@n61) {parameterName=UseReflection}
  n63 = StaticSwitchParameter(
   True=StaticSwitchParameter(True=StaticSwitchParameter(True=@n50, False=@n51) {parameterName=Variation_Only_Texture_2}, False=Multiply(A=@n48, B=@n62)) {parameterName=Variation_Only_Texture_1}
   False=@n62
  ) {parameterName=UseVariationTexture}
  n64 = VectorParameter() {parameterName=Specular_01_Color, defaultValue=<rgba 1 1 1 1>}
  n65 = Multiply(A=@n64.RGB, B=@n64.RGB)
  n66 = VectorParameter() {parameterName=Specular_02_Color, defaultValue=<rgba 1 1 1 2>}
  n67 = LinearInterpolate(
   A=Multiply(A=@n65, B=TextureSampleParameter2D(UVs=@n13) {parameterName=Specular_01, texture=&/T_Specular}.RGB) {desc=Specular 1}
   B=Multiply(
    A=Multiply(A=@n66.RGB, B=@n66.RGB)
    B=TextureSampleParameter2D(UVs=@n15) {parameterName=Specular_02, texture=&T_Specular}.RGB
   ) {desc=Specular 2}
   Alpha=@n32
  )
  n68 = Multiply(A=@n61, B=@n65)
  MaterialFunctionCall(
   DiffuseColor=StaticSwitchParameter(True=Multiply(A=@n35, B=@n63), False=@n63) {parameterName=UseFresnelRimLight}
   SpecularColor=Multiply(A=StaticSwitchParameter(True=@n67, False=@n68) {parameterName=UseSpecularTextures}) {constA=1, constB=0.04, desc=FresnelBaseReflectFraction}
  ) {materialFunction=&ConvertFromDiffSpec}
  Multiply(A=@n37.RGB, B=StaticSwitchParameter(True=Constant() {r=1}, False=Constant()) {parameterName=UseLightTransmission}) {desc=UE3 this connected to the Materials TransmissionMask input}
  n69 = TextureSampleParameter2D() {parameterName=OpacityMask, texture=&T_Reflection}
  StaticComponentMaskParameter(None=AppendVector(A=@n69.RGB, B=@n69.RGB)) {defaultR=true, parameterName=Opacity Mask Channel}
  n70 = TextureSampleParameter2D(UVs=Multiply(A=Multiply(A=@n1, B=@n12), B=ScalarParameter() {defaultValue=8, parameterName=DetailNormalScale})) {parameterName=DetailNormal, samplerType=SAMPLERTYPE_Normal, texture=&T_Normal}
  n71 = VectorParameter() {parameterName=DetailNormalIntensity, defaultValue=<rgba 2 2 0 1>}
  n72 = Add(
   A=LinearInterpolate(A=@n14.RGB, B=@n16.RGB, Alpha=@n32)
   B=StaticSwitchParameter(True=Multiply(A=@n70.RGB, B=@n71.RGB), False=Constant()) {parameterName=UseDetailNormal}
  )
  StaticSwitchParameter(True=Multiply(A=@n72, B=VectorParameter() {parameterName=NormalIntensity, defaultValue=<rgba 1 1 1 1>}.RGB), False=@n72) {parameterName=ModifyNormal}
  n73 = Constant()
  n74 = VectorParameter() {parameterName=EmissiveColor, defaultValue=<rgba 1 1 1 0.1>}
  n75 = Multiply(A=@n74.RGB, B=@n74.RGB)
  n76 = TextureSampleParameter2D() {parameterName=EmissiveTex, texture=&T_Emissive}
  StaticSwitchParameter(
   True=Add(A=StaticSwitchParameter(True=@n60, False=@n73) {parameterName=Reflection_Emissive}, B=Multiply(A=@n75, B=@n76.RGB))
   False=Constant()
  ) {parameterName=UseEmissive}
  n77 = VectorParameter() {parameterName=Transmission_Color, defaultValue=<rgba 1 1 1 1>}
  n78 = Multiply(A=@n77.RGB, B=@n77.RGB)
  StaticSwitchParameter(True=Multiply(A=@n34, B=@n78), False=@n78) {desc=UE3 this connected to the Materials TransmissionColor input, parameterName=LightTransmissionFresnel}
 ]
}
