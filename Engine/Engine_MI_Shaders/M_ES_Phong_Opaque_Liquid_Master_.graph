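# Translucent, two-sided, lit master material for liquid surfaces, despite "Opaque" in its name, with switchable features.
# Two panning normal maps, one optionally bump-offset, are summed to drive distortion and render-to-texture reflection offsets.
# Cubemap or render-to-texture reflection combines with emissive, edge color and optional diffuse texture, feeding a diffuse/specular conversion function.
# Depth-bias, fade and foreground-distortion branches appear as loose chains, several missing inputs, with no material output links shown.
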Material /Engine/Engine_MI_Shaders/M_ES_Phong_Opaque_Liquid_Master_
graph {
 properties {bUsedWithStaticLighting=true, blendMode=BLEND_Translucent, pixelDepthOffsetMode=PDOM_Legacy, twoSided=true, usedShadingModels=Default Lit}
 nodes [
  n1 = ScalarParameter() {defaultValue=2, parameterName=Scale}
  n2 = Multiply(A=TextureCoordinate(), B=@n1)
  n3 = ScalarParameter() {parameterName=BumpOffsetIntensity}
  n4 = TextureCoordinate()
  n5 = ScalarParameter() {defaultValue=2, parameterName=BumpOffsetScale}
  n6 = Time()
  n7 = ScalarParameter() {defaultValue=1, parameterName=BumpOffsetSpeed}
  n8 = TextureSampleParameter2D(UVs=Panner(Coordinate=Multiply(A=@n4, B=@n5), Time=Multiply(A=@n6, B=@n7)) {speedX=0.3, speedY=-0.3}) {parameterName=BumpOffsetTex, samplerType=SAMPLERTYPE_Normal, texture=&/T_Base_Tile_DetailNormal}
  n9 = StaticComponentMaskParameter(None=AppendVector(A=@n8.RGB, B=@n8.RGB)) {defaultR=true, parameterName=BumpOffsetChannel}
  n10 = StaticSwitchParameter(True=BumpOffset(Coordinate=@n2, Height=Multiply(A=@n3, B=@n9)), False=@n2) {parameterName=BumpOffsetNormal_1}
  n11 = Multiply(A=Time(), B=ScalarParameter() {defaultValue=1, parameterName=Speed})
  n12 = Multiply(A=TextureCoordinate() {uTiling=1.321, vTiling=1.321}, B=@n1)
  n13 = Add(
   A=TextureSampleParameter2D(UVs=Panner(Coordinate=@n10, Time=@n11) {speedX=0.3, speedY=-0.3}) {parameterName=Normal_01, samplerType=SAMPLERTYPE_Normal, texture=&T_Base_Tile_DetailNormal}.RGB
   B=TextureSampleParameter2D(UVs=Panner(Coordinate=@n12, Time=@n11) {speedX=0.5, speedY=0.25}) {parameterName=Normal_02, samplerType=SAMPLERTYPE_Normal, texture=&T_Base_Tile_DetailNormal}.RGB
  )
  n14 = Multiply(A=LinearInterpolate(A=VectorParameter() {parameterName=DistortionIntensity, defaultValue=<rgba 2 2 0 1>}.RGB, B=Constant()), B=@n13)
  n15 = ComponentMask(None=@n14) {g=true, r=true}
  n16 = Constant() {r=255}
  Add(A=ComponentMask(None=ScreenPosition().ViewportUV) {g=true, r=true}, B=Multiply(A=Constant2Vector() {g=-1, r=1}, B=Divide(A=@n15, B=@n16)))
  n17 = ScalarParameter() {defaultValue=0.5, parameterName=InnerDepthBias}
  n18 = ScalarParameter() {defaultValue=1, parameterName=Perspective Bias}
  n19 = Constant3Vector() {constant=<rgba 0 0 1 0>}
  n20 = StaticSwitchParameter(True=Abs(None=DotProduct(B=@n19)), False=Constant()) {parameterName=Use Perspective Correction}
  LinearInterpolate(A=@n17, B=Subtract(A=@n17, B=@n18), Alpha=@n20)
  n21 = ScalarParameter() {defaultValue=0.5, parameterName=OuterDepthBias}
  LinearInterpolate(A=@n21, B=Subtract(A=@n21, B=@n18), Alpha=@n20)
  MaterialFunctionCall!(Power=ScalarParameter() {defaultValue=5, parameterName=SpecPower}) {materialFunction=/Engine/Functions/Engine_MaterialFunctions01/Shading/PowerToRoughness}
  n22 = VectorParameter() {parameterName=CubeColor, defaultValue=<rgba 1 1 1 0.5>}
  n23 = Multiply(A=@n22.RGB, B=@n22.RGB)
  n24 = ComponentMask(None=ScreenPosition().ViewportUV) {g=true, r=true}
  n25 = VectorParameter() {parameterName=NormalIntensity, defaultValue=<rgba 1 1 1 1>}
  n26 = Multiply(
   A=ComponentMask(None=Multiply(A=@n13, B=@n25.RGB)) {g=true, r=true}
   B=ScalarParameter() {defaultValue=0.1, parameterName=R2TNormalIntensity}
  )
  n27 = VectorParameter() {parameterName=ReflectionNormal, defaultValue=<rgba 1 1 1 0>}
  n28 = TextureSampleParameterCube!(UVs=StaticSwitchParameter(True=Multiply(B=@n27.RGB), False=Constant()) {parameterName=ModifyReflectionNormal}, Apply View MipBias) {parameterName=CubeMap, texture=&/T_Cubemap_01}
  n29 = ScalarParameter() {defaultValue=1, parameterName=CubePower}
  n30 = StaticSwitchParameter(
   True=TextureSampleParameter2D(UVs=Add(A=@n24, B=@n26)) {parameterName=RenderToTextureReflection, texture=&/T_Reflection}.RGB
   False=StaticSwitchParameter(True=Power(Base=@n28.RGB, Exp=@n29), False=@n28.RGB) {parameterName=UseCubeMapPower}
  ) {parameterName=UseRenderToTextureReflection}
  n31 = VectorParameter() {parameterName=Emissive, defaultValue=<rgba 1 1 1 0>}
  n32 = VectorParameter() {parameterName=EdgeColor, defaultValue=<rgba 1 1 1 2>}
  n33 = Multiply(A=@n32.RGB, B=@n32.RGB)
  n34 = Constant() {r=1}
  n35 = Multiply(
   A=Add(A=Multiply(A=@n23, B=@n30), B=Multiply(A=@n31.RGB, B=@n31.RGB))
   B=StaticSwitchParameter(True=LinearInterpolate(A=@n33, B=@n34), False=@n34) {parameterName=BrightenEdges}
  )
  n36 = TextureCoordinate()
  n37 = ScalarParameter() {defaultValue=1, parameterName=DiffuseScale}
  n38 = VectorParameter() {parameterName=DiffuseTexColor, defaultValue=<rgba 1 1 1 2>}
  n39 = Multiply(
   A=TextureSampleParameter2D(UVs=Multiply(A=@n36, B=@n37)) {parameterName=DiffuseTexture, texture=&/T_Base_Tile_Variation}.RGB
   B=Multiply(A=@n38.RGB, B=@n38.RGB)
  )
  n40 = Multiply(A=OneMinus(), B=@n35)
  n41 = StaticSwitchParameter(
   True=StaticSwitchParameter(True=LinearInterpolate(A=@n35, B=@n39), False=Add(A=@n39, B=@n40)) {defaultValue=true, parameterName=BlendReflection}
   False=@n35
  ) {parameterName=UseDiffuseTexture}
  n42 = VectorParameter() {parameterName=SpecColor, defaultValue=<rgba 1 1 1 2>}
  n43 = Multiply(A=@n41, B=Multiply(A=@n42.RGB, B=@n42.RGB))
  n44 = Constant()
  MaterialFunctionCall(
   DiffuseColor=@n41
   SpecularColor=Multiply(A=StaticSwitchParameter(True=@n43, False=@n44) {defaultValue=true, parameterName=UseSpecular}) {constA=1, constB=0.04, desc=FresnelBaseReflectFraction}
  ) {materialFunction=&ConvertFromDiffSpec}
  n45 = TextureCoordinate()
  n46 = ScalarParameter() {defaultValue=1, parameterName=DiffuseAlphaScale}
  Multiply(
   A=TextureSampleParameter2D(UVs=Multiply(A=@n45, B=@n46)) {parameterName=DiffuseAlpha, texture=&T_Base_Tile_Variation}.RGB
   B=ScalarParameter() {defaultValue=1, parameterName=DiffuseAlphaBrightness}
  )
  Multiply(B=Clamp(Min=ScalarParameter() {parameterName=InnerBlendMin}))
  Multiply(A=PixelDepth().R, B=Divide(A=Constant() {r=0.0005}, B=ScalarParameter() {defaultValue=1, parameterName=DistortionFadeDistance}))
  StaticSwitchParameter(True=@n40, False=@n44) {parameterName=UseEmissiveCube}
  n47 = SceneDepth()
  n48 = PixelDepth()
  n49 = Constant()
  StaticSwitchParameter(
   True=StaticSwitchParameter(True=If(A=@n47.R, B=@n48.R, "A > B"=@n14, "A < B"=@n49), False=@n14) {defaultValue=true, parameterName=RemoveForegroundDistortionColor}
   False=Constant()
  ) {defaultValue=true, parameterName=UseDistortion}
 ]
}
