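# Translucent, default-lit master material that adapts Phong-style diffuse, specular and power inputs to PBR shading for material instances.
# Tinted, tiled diffuse and optional specular feed ConvertFromDiffSpec, while specular power goes through PowerToRoughness.
# Static switches control Fresnel rim lighting multiplied into diffuse, specular textures, scaled detail normals, normal intensity and textured emissive.
# Color parameters are multiplied by themselves, squaring their values before use.
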Material /Engine/Engine_MI_Shaders/M_ES_Phong_Translucent_Master_01
graph {
 properties {bUsedWithStaticLighting=true, blendMode=BLEND_Translucent, pixelDepthOffsetMode=PDOM_Legacy, usedShadingModels=Default Lit}
 nodes [
  MaterialFunctionCall!(Power=ScalarParameter() {defaultValue=5, parameterName=SpecPower1}) {materialFunction=/Engine/Functions/Engine_MaterialFunctions01/Shading/PowerToRoughness}
  n1 = Fresnel()
  n2 = ScalarParameter() {defaultValue=1, parameterName=FresnelFalloffExponent}
  n3 = VectorParameter() {desc=Color, parameterName=FresnelColor, defaultValue=<rgba 1 1 1 5>}
  n4 = Add(A=Multiply(A=Power(Base=@n1, Exp=@n2), B=Multiply(A=@n3.RGB, B=@n3.RGB)), B=Constant() {r=1})
  n5 = VectorParameter() {parameterName=Diffuse_01_Color, defaultValue=<rgba 1 1 1 1>}
  n6 = TextureCoordinate()
  n7 = Multiply(A=@n6, B=ScalarParameter() {defaultValue=1, parameterName=Tex_Scale_01})
  n8 = Multiply(
   A=Multiply(A=@n5.RGB, B=@n5.RGB)
   B=TextureSampleParameter2D(UVs=@n7) {parameterName=Diffuse_01, texture=&/T_Reflection}.RGB
  )
  n9 = VectorParameter() {parameterName=Specular_01_Color, defaultValue=<rgba 1 1 1 1>}
  n10 = Multiply(A=@n9.RGB, B=@n9.RGB)
  n11 = StaticSwitchParameter(
   True=Multiply(A=@n10, B=TextureSampleParameter2D(UVs=@n7) {parameterName=Specular_01, texture=&/T_Specular}.RGB)
   False=Multiply(A=@n8, B=@n10)
  ) {parameterName=UseSpecularTextures}
  n12 = Constant()
  MaterialFunctionCall(
   DiffuseColor=StaticSwitchParameter(True=Multiply(A=@n4, B=@n8), False=@n8) {parameterName=UseFresnelRimLight}
   SpecularColor=Multiply(A=StaticSwitchParameter(True=@n11, False=@n12) {defaultValue=true, parameterName=UseSpecular}) {constA=1, constB=0.04, desc=FresnelBaseReflectFraction}
  ) {materialFunction=&ConvertFromDiffSpec}
  ScalarParameter() {defaultValue=0.5, parameterName=DepthBias}
  n13 = TextureCoordinate()
  n14 = TextureSampleParameter2D(UVs=Multiply(A=@n6, B=ScalarParameter() {defaultValue=8, parameterName=DetailNormalScale})) {parameterName=DetailNormal, samplerType=SAMPLERTYPE_Normal, texture=&/T_Normal}
  n15 = VectorParameter() {parameterName=DetailNormalIntensity, defaultValue=<rgba 2 2 0 1>}
  n16 = Add(
   A=TextureSampleParameter2D(UVs=StaticSwitchParameter(True=@n7, False=@n13) {defaultValue=true, parameterName=ScaleNormal_01}) {parameterName=Normal_01, samplerType=SAMPLERTYPE_Normal, texture=&T_Normal}.RGB
   B=StaticSwitchParameter(True=Multiply(A=@n14.RGB, B=@n15.RGB), False=Constant()) {parameterName=UseDetailNormal}
  )
  StaticSwitchParameter(True=Multiply(A=@n16, B=VectorParameter() {parameterName=NormalIntensity, defaultValue=<rgba 1 1 1 1>}.RGB), False=@n16) {parameterName=ModifyNormal}
  n17 = VectorParameter() {parameterName=EmissiveColor, defaultValue=<rgba 1 1 1 0.1>}
  StaticSwitchParameter(
   True=Multiply(
    A=Multiply(A=@n17.RGB, B=@n17.RGB)
    B=TextureSampleParameter2D() {parameterName=EmissiveTex, texture=&/T_Emissive}.RGB
   )
   False=Constant()
  ) {parameterName=UseEmissive}
 ]
}
