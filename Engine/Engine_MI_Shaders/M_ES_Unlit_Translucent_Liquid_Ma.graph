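# Unlit, translucent, two-sided master material for liquid surfaces, with tunable color, opacity, reflection and distortion.
# Two panned normal maps, one optionally bump-offset, are summed to perturb screen UVs of an optional render-to-texture reflection.
# That reflection is power-adjusted, tinted by squared CubeColor and added to squared Emissive, optionally feeding opacity.
# The graph also holds edge-brightening, depth-bias and distance-fade distortion terms, though some of their inputs and outputs appear unconnected.
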Material /Engine/Engine_MI_Shaders/M_ES_Unlit_Translucent_Liquid_Ma
graph {
 properties {blendMode=BLEND_Translucent, pixelDepthOffsetMode=PDOM_Legacy, shadingModel=MSM_Unlit, twoSided=true, usedShadingModels=Unlit}
 nodes [
  n1 = Constant() {r=1}
  n2 = VectorParameter() {parameterName=CubeColor, defaultValue=<rgba 1 1 1 0.5>}
  n3 = ComponentMask(None=ScreenPosition().ViewportUV) {g=true, r=true}
  n4 = ScalarParameter() {defaultValue=2, parameterName=Scale}
  n5 = Multiply(A=TextureCoordinate(), B=@n4)
  n6 = ScalarParameter() {parameterName=BumpOffsetIntensity}
  n7 = TextureCoordinate()
  n8 = ScalarParameter() {defaultValue=2, parameterName=BumpOffsetScale}
  n9 = Time()
  n10 = ScalarParameter() {defaultValue=1, parameterName=BumpOffsetSpeed}
  n11 = TextureSampleParameter2D(UVs=Panner(Coordinate=Multiply(A=@n7, B=@n8), Time=Multiply(A=@n9, B=@n10)) {speedX=0.3, speedY=-0.3}) {parameterName=BumpOffsetTex, samplerType=SAMPLERTYPE_Normal, texture=&/T_Base_Tile_DetailNormal}
  n12 = StaticComponentMaskParameter(None=AppendVector(A=@n11.RGB, B=@n11.RGB)) {defaultR=true, parameterName=BumpOffsetChannel}
  n13 = StaticSwitchParameter(True=BumpOffset(Coordinate=@n5, Height=Multiply(A=@n6, B=@n12)), False=@n5) {parameterName=BumpOffsetNormal_1}
  n14 = Multiply(A=Time(), B=ScalarParameter() {defaultValue=1, parameterName=Speed})
  n15 = Multiply(A=TextureCoordinate() {uTiling=1.321, vTiling=1.321}, B=@n4)
  n16 = Add(
   A=TextureSampleParameter2D(UVs=Panner(Coordinate=@n13, Time=@n14) {speedX=0.3, speedY=-0.3}) {parameterName=Normal_01, samplerType=SAMPLERTYPE_Normal, texture=&T_Base_Tile_DetailNormal}.RGB
   B=TextureSampleParameter2D(UVs=Panner(Coordinate=@n15, Time=@n14) {speedX=0.5, speedY=0.25}) {parameterName=Normal_02, samplerType=SAMPLERTYPE_Normal, texture=&T_Base_Tile_DetailNormal}.RGB
  )
  n17 = VectorParameter() {parameterName=NormalIntensity, defaultValue=<rgba 1 1 1 1>}
  n18 = Multiply(
   A=ComponentMask(None=Multiply(A=@n16, B=@n17.RGB)) {g=true, r=true}
   B=ScalarParameter() {defaultValue=0.1, parameterName=R2TNormalIntensity}
  )
  n19 = StaticSwitchParameter(
   True=TextureSampleParameter2D(UVs=Add(A=@n3, B=@n18)) {parameterName=RenderToTextureReflection, texture=&/T_Reflection}.RGB
   False=Constant() {r=0.5}
  ) {parameterName=UseRenderToTextureReflection}
  n20 = ScalarParameter() {defaultValue=1, parameterName=CubePower}
  n21 = VectorParameter() {parameterName=Emissive, defaultValue=<rgba 1 1 1 0>}
  n22 = Add(A=Multiply(A=Multiply(A=@n2.RGB, B=@n2.RGB), B=Power(Base=@n19, Exp=@n20)), B=Multiply(A=@n21.RGB, B=@n21.RGB))
  n23 = ComponentMask(None=@n22) {r=true}
  n24 = ScalarParameter() {defaultValue=0.5, parameterName=ReflectionOpacityContribution}
  n25 = Clamp(Min=ScalarParameter() {parameterName=InnerBlendMin})
  Multiply(
   A=StaticSwitchParameter(True=LinearInterpolate(A=@n1, B=@n23, Alpha=@n24), False=@n1) {parameterName=UseReflectionInOpacity}
   B=Multiply(A=Multiply(B=@n25), B=ScalarParameter() {defaultValue=0.9, parameterName=Opacity})
  )
  Multiply(A=LinearInterpolate(A=VectorParameter() {parameterName=DistortionIntensity, defaultValue=<rgba 2 2 0 1>}.RGB, B=Constant()), B=@n16)
  n26 = VectorParameter() {parameterName=EdgeColor, defaultValue=<rgba 1 1 1 2>}
  n27 = Multiply(A=@n26.RGB, B=@n26.RGB)
  n28 = Constant() {r=1}
  Multiply(A=@n22, B=StaticSwitchParameter(True=LinearInterpolate(A=@n27, B=@n28), False=@n28) {parameterName=BrightenEdges})
  Multiply(A=PixelDepth().R, B=Divide(A=Constant() {r=0.0005}, B=ScalarParameter() {defaultValue=1, parameterName=DistortionFadeDistance}))
  ScalarParameter() {defaultValue=0.5, parameterName=InnerDepthBias}
  ScalarParameter() {defaultValue=0.5, parameterName=OuterDepthBias}
 ]
}
