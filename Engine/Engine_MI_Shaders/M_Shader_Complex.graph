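# Parameterized Default Lit master material in engine content, layering diffuse, specular, normal, emissive, masked reflection and variation textures.
# A rotated, scaled variation texture is powered and color-blended to modulate diffuse, and separately scales specular.
# Masked reflection, sampled at constant UVs rather than a reflection vector, adds to diffuse before ConvertFromDiffSpec; PowerToRoughness converts specular power.
# A rotated, tiled detail normal adds to the base normal, and tinted emissive is separate.
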Material /Engine/Engine_MI_Shaders/M_Shader_Complex
graph {
 properties {bUsedWithStaticLighting=true, pixelDepthOffsetMode=PDOM_Legacy, usedShadingModels=Default Lit}
 nodes [
  n1 = Rotator(
   Coordinate=Multiply(A=ScalarParameter() {defaultValue=20, desc=DetailNormalScale, parameterName=DetailNormalScale}, B=TextureCoordinate())
   Time=ScalarParameter() {parameterName=DetailNormalRotate}
  )
  Add(
   A=Multiply(
    A=TextureSampleParameter2D() {desc=Normal, parameterName=Normal, samplerType=SAMPLERTYPE_Normal, texture=&/T_Normal}.RGB
    B=VectorParameter() {desc=Normal multiply, parameterName=NormalMultiply, defaultValue=<rgba 1 1 1 0>}.RGB
   )
   B=Multiply(
    A=TextureSampleParameter2D(UVs=@n1) {desc=DetailNormal, parameterName=DetailNormal, samplerType=SAMPLERTYPE_Normal, texture=&T_Normal}.RGB
    B=VectorParameter() {desc=Normal multiply, parameterName=NormalDetailMultiply}.RGB
   )
  )
  MaterialFunctionCall!(Power=ScalarParameter() {defaultValue=5, desc=SpecularPower, parameterName=SpecPower}) {materialFunction=/Engine/Functions/Engine_MaterialFunctions01/Shading/PowerToRoughness}
  n2 = Constant()
  n3 = Multiply(
   A=TextureSampleParameter2D(UVs=ComponentMask(None=@n2) {g=true, r=true}) {desc=Reflection, parameterName=reflection, texture=&/T_Emissive}.RGB
   B=ScalarParameter() {defaultValue=1, desc=VariationReflection, parameterName=ReflectionMultiply}
  )
  n4 = TextureSampleParameter2D() {desc=ReflectionMask, parameterName=ReflectionMask, texture=&T_Emissive}
  n5 = TextureSampleParameter2D() {desc=Diffuse, parameterName=Diffuse, texture=&/T_Reflection}
  n6 = Multiply(
   A=ScalarParameter() {defaultValue=0.3, desc=VariationScale, parameterName=VariationScale}
   B=Rotator(Coordinate=TextureCoordinate(), Time=ScalarParameter() {desc=RotationVariable, parameterName=VariationRotate})
  )
  n7 = Power(
   Base=Add(
    A=TextureSampleParameter2D(UVs=@n6) {desc=VariationTex, parameterName=VariationTexture, texture=&T_Reflection}.RGB
    B=Constant() {r=0.004}
   )
   Exp=ScalarParameter() {defaultValue=1, desc=VariationPower, parameterName=VariationPower}
  )
  n8 = Multiply(A=@n7, B=ScalarParameter() {defaultValue=100, desc=VariationMultiply, parameterName=VariationMultiply})
  n9 = Clamp(None=OneMinus(None=@n8), Min=Constant(), Max=Constant() {r=1})
  n10 = LinearInterpolate(
   A=Clamp(None=@n8, Min=Constant() {r=0.5}, Max=Constant() {r=5})
   B=Multiply(A=VectorParameter() {desc=Color, parameterName=VariationColor}.RGB, B=@n9)
   Alpha=@n9
  )
  n11 = Multiply(A=@n7, B=ScalarParameter() {defaultValue=12, desc=VariationSpecularMult, parameterName=VariationSpecularMult})
  n12 = TextureSampleParameter2D() {desc=Specular, parameterName=Specular, texture=&T_Emissive}
  MaterialFunctionCall(
   DiffuseColor=Add(A=Multiply(A=@n3, B=@n4.RGB), B=Multiply(A=@n5.RGB, B=@n10))
   SpecularColor=Multiply(A=Multiply(A=@n11, B=@n12.RGB)) {constA=1, constB=0.04, desc=FresnelBaseReflectFraction}
  ) {materialFunction=&ConvertFromDiffSpec}
  Multiply(
   A=TextureSampleParameter2D() {desc=Emissive, parameterName=Emissive, texture=&T_Emissive}.RGB
   B=VectorParameter() {desc=Emissivecolor, parameterName=EmissiveColor}.RGB
  )
 ]
}
