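# M_Shader_Medium is a Default Lit master material in the engine's MI shader set, exposing texture and scalar parameters for instances.
# It adds a rotatable, scaled detail normal to the base normal, and converts masked specular power to roughness.
# Diffuse blends toward a constant-UV reflection texture by mask, then feeds ConvertFromDiffSpec with scaled specular.
# Emissive is a multiplied texture, with a static switch adding a pink debug tint.
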Material /Engine/Engine_MI_Shaders/M_Shader_Medium
graph {
 properties {bUsedWithStaticLighting=true, pixelDepthOffsetMode=PDOM_Legacy, usedShadingModels=Default Lit}
 nodes [
  n1 = Rotator(
   Coordinate=Multiply(A=TextureCoordinate(), B=ScalarParameter() {defaultValue=1, parameterName=NormalDetailScale})
   Time=ScalarParameter() {parameterName=DetailNormalRotate}
  )
  Add(
   A=Multiply(
    A=TextureSampleParameter2D() {parameterName=Normal, samplerType=SAMPLERTYPE_Normal, texture=&/T_Normal}.RGB
    B=VectorParameter() {parameterName=NormalMultiply, defaultValue=<rgba 1 1 1 1>}.RGB
   )
   B=Multiply(
    A=TextureSampleParameter2D(UVs=@n1) {parameterName=NormalDetail, samplerType=SAMPLERTYPE_Normal, texture=&T_Normal}.RGB
    B=VectorParameter() {parameterName=NormalDetailMultiply, defaultValue=<rgba 1 1 1 1>}.RGB
   )
  )
  n2 = TextureSampleParameter2D() {parameterName=SpecularPowerMask, texture=&/T_Emissive}
  MaterialFunctionCall!(Power=LinearInterpolate(
    A=ScalarParameter() {defaultValue=5, parameterName=SpecPowerLow}
    B=ScalarParameter() {defaultValue=50, parameterName=SpecPowerHigh}
    Alpha=ComponentMask(None=@n2.RGB) {g=true}
   )) {materialFunction=/Engine/Functions/Engine_MaterialFunctions01/Shading/PowerToRoughness}
  n3 = TextureSampleParameter2D(UVs=ComponentMask(None=Constant()) {g=true, r=true}) {parameterName=reflection, texture=&/T_Reflection}
  n4 = ScalarParameter() {defaultValue=10, parameterName=EnvMultiply}
  n5 = TextureSampleParameter2D() {parameterName=ReflectionMask, texture=&T_Emissive}
  n6 = TextureSampleParameter2D() {parameterName=Spec, texture=&/T_Specular}
  n7 = ScalarParameter() {defaultValue=1, parameterName=SpecularMultiply}
  MaterialFunctionCall(
   DiffuseColor=LinearInterpolate(
    A=TextureSampleParameter2D() {desc=Diffuse, parameterName=Diffuse, texture=&T_Reflection}.RGB
    B=Multiply(A=@n3.RGB, B=@n4)
    Alpha=ComponentMask(None=@n5.RGB) {r=true}
   )
   SpecularColor=Multiply(A=Multiply(A=@n6.RGB, B=@n7)) {constA=1, constB=0.04, desc=FresnelBaseReflectFraction}
  ) {materialFunction=&ConvertFromDiffSpec}
  n8 = Multiply(
   A=TextureSampleParameter2D() {parameterName=Emissive, texture=&T_Emissive}.RGB
   B=ScalarParameter() {defaultValue=10, parameterName=EmissiveMultiply}
  )
  StaticSwitchParameter(True=Add(A=Constant3Vector() {constant=<rgba 2 0 2 0>}, B=@n8), False=@n8) {parameterName=UsePinkDebug}
 ]
}
