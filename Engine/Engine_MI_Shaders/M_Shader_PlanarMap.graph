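# Lit master material that planar-maps diffuse, normal and specular textures in world space along the three axes, independent of mesh UVs.
# World position planes XZ, XY and YZ, scaled by U/V Scale parameters, provide sampling UVs.
# The transformed surface normal's absolute axis components are sharpened by multiply-clamp-subtract into weights for nested lerps between projections.
# ConvertFromDiffSpec converts blended diffuse and specular, with specular gated by the UseSpecular switch.
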Material /Engine/Engine_MI_Shaders/M_Shader_PlanarMap
graph {
 properties {bUsedWithStaticLighting=true, pixelDepthOffsetMode=PDOM_Legacy, usedShadingModels=Default Lit}
 nodes [
  n1 = Constant() {desc=Base scaling, r=0.0039}
  n2 = Multiply(
   A=AppendVector(A=@n1, B=@n1)
   B=AppendVector(A=ScalarParameter() {defaultValue=1, parameterName=U Scale}, B=ScalarParameter() {defaultValue=1, parameterName=V Scale})
  )
  n3 = WorldPosition()
  n4 = Multiply(A=@n2, B=ComponentMask(None=@n3.XYZ) {b=true, desc=Filter Y Axis, r=true})
  n5 = TextureSampleParameter2D(UVs=@n4) {parameterName=NormalTexture, samplerType=SAMPLERTYPE_Normal, texture=&/T_Normal}
  n6 = Multiply(A=@n2, B=ComponentMask(None=@n3.XYZ) {desc=Filter Z Axis, g=true, r=true})
  n7 = Transform(None=Constant3Vector() {constant=<rgba 0 0 1 0>})
  n8 = Abs(None=DotProduct(A=@n7, B=Constant3Vector() {desc=Z, constant=<rgba 0 0 1 0>}))
  n9 = Constant() {r=3}
  n10 = Constant() {r=2}
  n11 = Subtract(A=Clamp(None=Multiply(A=@n8, B=@n9), Min=@n10, Max=@n9), B=@n10)
  n12 = Multiply(A=@n2, B=ComponentMask(None=@n3.XYZ) {b=true, desc=Filter X Axis, g=true})
  n13 = Abs(None=DotProduct(A=@n7, B=Constant3Vector() {desc=Y, constant=<rgba 0 1 0 0>}))
  n14 = Subtract(A=Clamp(None=Multiply(A=@n13, B=@n9), Min=@n10, Max=@n9), B=@n10)
  n15 = Abs(None=DotProduct(A=@n7, B=Constant3Vector() {desc=X, constant=<rgba 1 0 0 0>}))
  n16 = Subtract(A=Clamp(None=Multiply(A=@n15, B=@n9), Min=@n10, Max=@n9), B=@n10)
  LinearInterpolate(
   A=LinearInterpolate(
    A=@n5.RGB
    B=TextureSampleParameter2D(UVs=@n6) {parameterName=NormalTexture, samplerType=SAMPLERTYPE_Normal, texture=&T_Normal}.RGB
    Alpha=@n11
   ) {desc=YZ}
   B=LinearInterpolate(
    A=TextureSampleParameter2D(UVs=@n12) {parameterName=NormalTexture, samplerType=SAMPLERTYPE_Normal, texture=&T_Normal}.RGB
    B=@n5.RGB
    Alpha=@n14
   ) {desc=XY}
   Alpha=@n16
  )
  n17 = TextureSampleParameter2D(UVs=@n4) {parameterName=DiffuseTexture, texture=&/T_Reflection}
  n18 = TextureSampleParameter2D(UVs=@n6) {parameterName=DiffuseTexture, texture=&T_Reflection}
  n19 = TextureSampleParameter2D(UVs=@n12) {parameterName=DiffuseTexture, texture=&T_Reflection}
  n20 = TextureSampleParameter2D(UVs=@n4) {parameterName=SpecularTexture, texture=&/T_Specular}
  n21 = LinearInterpolate(
   A=LinearInterpolate(
    A=@n20.RGB
    B=TextureSampleParameter2D(UVs=@n6) {parameterName=SpecularTexture, texture=&T_Specular}.RGB
    Alpha=@n11
   ) {desc=YZ}
   B=LinearInterpolate(
    A=TextureSampleParameter2D(UVs=@n12) {parameterName=SpecularTexture, texture=&T_Specular}.RGB
    B=@n20.RGB
    Alpha=@n14
   ) {desc=XY}
   Alpha=@n16
  )
  n22 = Constant()
  MaterialFunctionCall(
   DiffuseColor=LinearInterpolate(
    A=LinearInterpolate(A=@n17.RGB, B=@n18.RGB, Alpha=@n11) {desc=YZ}
    B=LinearInterpolate(A=@n19.RGB, B=@n17.RGB, Alpha=@n14) {desc=XY}
    Alpha=@n16
   )
   SpecularColor=Multiply(A=StaticSwitchParameter(True=@n21, False=@n22) {parameterName=UseSpecular}) {constA=1, constB=0.04, desc=FresnelBaseReflectFraction}
  ) {materialFunction=&ConvertFromDiffSpec}
 ]
}
