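# Two-sided, masked Default Lit base material whose textures and multipliers are exposed as parameters.
# Diffuse and scaled Spec textures pass through ConvertFromDiffSpec, and a specular power parameter through PowerToRoughness, translating a diffuse/specular setup to PBR values.
# A tinted normal map is added to a tinted detail normal with scalable, rotatable UVs.
# An emissive texture has a brightness multiplier, and a static component mask selects the opacity mask channel.
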Material /Engine/Engine_MI_Shaders/M_Shader_SimpleMaskedTwoSided
graph {
 properties {bUsedWithStaticLighting=true, blendMode=BLEND_Masked, pixelDepthOffsetMode=PDOM_Legacy, twoSided=true, usedShadingModels=Default Lit}
 nodes [
  n1 = Rotator(
   Coordinate=Multiply(A=TextureCoordinate(), B=ScalarParameter() {defaultValue=1, parameterName=NormalDetailScale})
   Time=ScalarParameter() {parameterName=NormalDetailRotate}
  )
  Add(
   A=Multiply(
    A=TextureSampleParameter2D() {parameterName=Normal, samplerType=SAMPLERTYPE_Normal, texture=&/T_Normal}.RGB
    B=VectorParameter() {parameterName=NormalMultiply, defaultValue=<rgba 1 1 1 1>}.RGB
   )
   B=Multiply(
    A=TextureSampleParameter2D(UVs=@n1) {parameterName=NormalDetail, samplerType=SAMPLERTYPE_Normal, texture=&T_Normal}.RGB
    B=VectorParameter() {parameterName=NormalDetailMultiply, defaultValue=<rgba 1 1 1 1>}.RGB
   )
  )
  MaterialFunctionCall!(Power=ScalarParameter() {defaultValue=15, desc=Specular Power, parameterName=SpecPower}) {materialFunction=/Engine/Functions/Engine_MaterialFunctions01/Shading/PowerToRoughness}
  n2 = TextureSampleParameter2D() {parameterName=Spec, texture=&/T_Reflection}
  n3 = ScalarParameter() {defaultValue=1, parameterName=SpecularMultValue}
  MaterialFunctionCall(
   DiffuseColor=TextureSampleParameter2D() {desc=Diffuse, parameterName=Diffuse, texture=&T_Reflection}.RGB
   SpecularColor=Multiply(A=Multiply(A=@n2.RGB, B=@n3)) {constA=1, constB=0.04, desc=FresnelBaseReflectFraction}
  ) {materialFunction=&ConvertFromDiffSpec}
  Multiply(
   A=TextureSampleParameter2D() {parameterName=Emissive, texture=&/T_Emissive}.RGB
   B=ScalarParameter() {defaultValue=10, parameterName=EmissiveMultiply}
  )
  n4 = TextureSampleParameter2D() {parameterName=OpacityMask, texture=&T_Reflection}
  StaticComponentMaskParameter(None=AppendVector(A=@n4.RGB, B=@n4.RGB)) {defaultR=true, parameterName=Opacity Mask Channel}
 ]
}
