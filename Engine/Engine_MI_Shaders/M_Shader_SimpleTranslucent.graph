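# Simple unlit, two-sided translucent base material in the engine's MI shaders folder, exposing texture slots for color and opacity.
# It declares two texture parameters, Diffuse and OpacityMask, both defaulting to T_Reflection, with no math nodes or visible wiring to material outputs.
# Usage flags enable it for skeletal meshes, mesh deformers and static lighting.
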Material /Engine/Engine_MI_Shaders/M_Shader_SimpleTranslucent
graph {
 properties {
  bUsedWithMeshDeformer=true
  bUsedWithSkeletalMesh=true
  bUsedWithStaticLighting=true
  blendMode=BLEND_Translucent
  pixelDepthOffsetMode=PDOM_Legacy
  shadingModel=MSM_Unlit
  twoSided=true
  usedShadingModels=Unlit
 }
 nodes [
  TextureSampleParameter2D() {desc=Diffuse, parameterName=Diffuse, texture=&/T_Reflection}
  TextureSampleParameter2D() {parameterName=OpacityMask, texture=&T_Reflection}
 ]
}
