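# M_Shader_Simple_CustomUV is a Default Lit master material with parameterized diffuse, specular, normal and emissive textures for static-lighting use.
# A static switch, on by default, applies separate U and V scales to the UVs shared by all maps.
# The base normal adds a scaled, rotatable detail normal, each with a color multiplier.
# ConvertFromDiffSpec converts diffuse and scaled specular, PowerToRoughness converts specular power, and emissive gets a multiplier.
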Material /Engine/Engine_MI_Shaders/M_Shader_Simple_CustomUV
graph {
 properties {bUsedWithStaticLighting=true, pixelDepthOffsetMode=PDOM_Legacy, usedShadingModels=Default Lit}
 nodes [
  n1 = ScalarParameter() {defaultValue=1, parameterName=Scale U}
  n2 = TextureCoordinate()
  n3 = ComponentMask(None=@n2) {r=true}
  n4 = ComponentMask(None=@n2) {g=true}
  n5 = ScalarParameter() {defaultValue=1, parameterName=Scale V}
  n6 = StaticSwitchParameter(True=AppendVector(A=Multiply(A=@n1, B=@n3), B=Multiply(A=@n4, B=@n5)), False=@n2) {defaultValue=true, parameterName=Use Custom UVs}
  n7 = Rotator(
   Coordinate=Multiply(A=@n6, B=ScalarParameter() {defaultValue=1, parameterName=NormalDetailScale})
   Time=ScalarParameter() {parameterName=DetailNormalRotate}
  )
  Add(
   A=Multiply(
    A=TextureSampleParameter2D(UVs=@n6) {parameterName=Normal, samplerType=SAMPLERTYPE_Normal, texture=&/T_Normal}.RGB
    B=VectorParameter() {parameterName=NormalMultiply, defaultValue=<rgba 1 1 1 1>}.RGB
   )
   B=Multiply(
    A=TextureSampleParameter2D(UVs=@n7) {parameterName=NormalDetail, samplerType=SAMPLERTYPE_Normal, texture=&T_Normal}.RGB
    B=VectorParameter() {parameterName=NormalDetailMultiply, defaultValue=<rgba 1 1 1 1>}.RGB
   )
  )
  MaterialFunctionCall!(Power=ScalarParameter() {defaultValue=15, parameterName=SpecPower}) {materialFunction=/Engine/Functions/Engine_MaterialFunctions01/Shading/PowerToRoughness}
  n8 = TextureSampleParameter2D(UVs=@n6) {parameterName=Spec, texture=&/T_Reflection}
  n9 = ScalarParameter() {defaultValue=1, parameterName=SpecularMultValue}
  MaterialFunctionCall(
   DiffuseColor=TextureSampleParameter2D(UVs=@n6) {parameterName=Diffuse, texture=&T_Reflection}.RGB
   SpecularColor=Multiply(A=Multiply(A=@n8.RGB, B=@n9)) {constA=1, constB=0.04, desc=FresnelBaseReflectFraction}
  ) {materialFunction=&ConvertFromDiffSpec}
  Multiply(
   A=TextureSampleParameter2D(UVs=@n6) {parameterName=Emissive, texture=&/T_Emissive}.RGB
   B=ScalarParameter() {defaultValue=10, parameterName=EmissiveMultiply}
  )
 ]
}
